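# Physics asset for the F_MED_RockerPunkAlt female character's FaceAcc mesh, setting up mirrored hair strands and ponytails for physics.
# Capsule bodies cover the hair bones plus head, neck, upper spine, clavicles and upper arms.
# Dynamic hair forms short chains with tight swing and twist limits, hung from the head or base hair bones.
# Body links and hair-A base-to-head joints leave rotation free; hair-B and ponytail base bodies have no constraints.
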PhysicsAsset /BRCosmetics/Characters/Player/Female/Medium/Bodies/F_MED_RockerPunkAlt/Meshes/Parts/F_MED_RockerPunkAlt_FaceAcc_Physics_Main
bodies {
 base_hair_a_01_l [{shapeType=Capsule, length=1, radius=1, shapeName=base_hair_a_01_l_capsule, center=<1 0 0>, rotation=<rot 0 -90 0>}]
 base_hair_a_01_r [{shapeType=Capsule, length=1, radius=1, shapeName=base_hair_a_01_r_capsule, center=<1 0 0>, rotation=<rot 0 -90 0>}]
 base_hair_b_01_l [{shapeType=Capsule, length=0.505, radius=0.505, shapeName=base_hair_b_01_l_capsule, rotation=<rot 0 -90 0>}]
 base_hair_b_01_r [{shapeType=Capsule, length=0.505, radius=0.505, shapeName=base_hair_b_01_r_capsule, rotation=<rot 0 -90 0>}]
 base_ponytail_00_l [{shapeType=Capsule, length=0.505, radius=0.505, shapeName=base_ponytail_00_l_capsule, rotation=<rot 0 -90 0>}]
 base_ponytail_00_r [{shapeType=Capsule, length=0.505, radius=0.505, shapeName=base_ponytail_00_r_capsule, rotation=<rot 0 -90 0>}]
 clavicle_l [{shapeType=Capsule, length=6.25, radius=6, shapeName=clavicle_l_capsule, center=<6 -2 -1.5>, rotation=<rot -90 0 0>}]
 clavicle_r [{shapeType=Capsule, length=6.25, radius=6, shapeName=clavicle_r_capsule, center=<-6 2 1.5>, rotation=<rot -90 0 0>}]
 dyn_hair_a_01_l [{shapeType=Capsule, length=2.5, radius=1.5, shapeName=dyn_hair_a_01_l_capsule, center=<5 0 0>, rotation=<rot -90 0 0>}]
 dyn_hair_a_01_r [{shapeType=Capsule, length=2.5, radius=1.5, shapeName=dyn_hair_a_01_r_capsule, center=<5 0 0>, rotation=<rot -90 0 0>}]
 dyn_hair_a_02_l [{shapeType=Capsule, length=4.75, radius=1.75, shapeName=dyn_hair_a_02_l_capsule, center=<5 0 0>, rotation=<rot -90 0 0>}]
 dyn_hair_a_02_r [{shapeType=Capsule, length=4.75, radius=1.75, shapeName=dyn_hair_a_02_r_capsule, center=<5 0 0>, rotation=<rot -90 0 0>}]
 dyn_hair_b_01_l [{shapeType=Capsule, length=2.25, radius=1.75, shapeName=dyn_hair_b_01_l_capsule, center=<2.0336 0 0>, rotation=<rot -90 180 180>}]
 dyn_hair_b_01_r [{shapeType=Capsule, length=2.25, radius=1.75, shapeName=dyn_hair_b_01_r_capsule, center=<2.0336 0 0>, rotation=<rot -90 0 0>}]
 dyn_hair_b_02_l [{shapeType=Capsule, length=5, radius=2, shapeName=dyn_hair_b_02_l_capsule, center=<4.9888 0 0>, rotation=<rot -90 180 180>}]
 dyn_hair_b_02_r [{shapeType=Capsule, length=5, radius=2, shapeName=dyn_hair_b_02_l_capsule, center=<4.9888 0 0>, rotation=<rot -90 180 180>}]
 dyn_hair_fr_l [{shapeType=Capsule, length=1, radius=1, shapeName=dyn_hair_fr_l_capsule, center=<1 0 0>, rotation=<rot -90 0 0>}]
 dyn_hair_fr_r [{shapeType=Capsule, length=1, radius=1, shapeName=dyn_hair_fr_r_capsule, center=<1 0 0>, rotation=<rot -90 0 0>}]
 dyn_ponytail_00_l [{shapeType=Capsule, length=1, radius=1, shapeName=dyn_ponytail_00_l_capsule, center=<1 0 0>, rotation=<rot -90 0 0>}]
 dyn_ponytail_00_r [{shapeType=Capsule, length=1, radius=1, shapeName=dyn_ponytail_00_r_capsule, center=<1 0 0>, rotation=<rot -90 0 0>}]
 dyn_ponytail_01_l [{shapeType=Capsule, length=1, radius=1, shapeName=dyn_ponytail_01_l_capsule, center=<1.5 0 0>, rotation=<rot -90 0 0>}]
 dyn_ponytail_01_r [{shapeType=Capsule, length=1, radius=1, shapeName=dyn_ponytail_01_r_capsule, center=<1.5 0 0>, rotation=<rot -90 0 0>}]
 dyn_ponytail_02_l [{shapeType=Capsule, length=1, radius=1, shapeName=dyn_ponytail_02_l_capsule, center=<3 0 0>, rotation=<rot -90 0 0>}]
 dyn_ponytail_02_r [{shapeType=Capsule, length=1, radius=1, shapeName=dyn_ponytail_02_r_capsule, center=<3 0 0>, rotation=<rot -90 0 0>}]
 head [{shapeType=Capsule, length=3, radius=6.8, shapeName=head_capsule, center=<1 0 0>, rotation=<rot 90 0 0>}]
 neck_01 [{shapeType=Capsule, length=5.25, radius=4.75, shapeName=neck_01_capsule, center=<1 0 0>, rotation=<rot -90 0 0>}]
 neck_02 [{shapeType=Capsule, length=1, radius=1, shapeName=neck_02_capsule, rotation=<rot -90 0 0>}]
 spine_05 [{shapeType=Capsule, length=1, radius=1, shapeName=spine_05_capsule, rotation=<rot -90 0 0>}]
 upperarm_l [{shapeType=Capsule, length=10, radius=5.5, shapeName=upperarm_l_capsule, center=<9 0 0>, rotation=<rot -90 0 0>}]
 upperarm_r [{shapeType=Capsule, length=10, radius=5.5, shapeName=upperarm_r_capsule, center=<-9 0 0>, rotation=<rot -90 0 0>}]
}
constraints [
 {bone1Name=dyn_ponytail_01_r, bone2Name=head, swing1LimitDegrees=45, swing2LimitDegrees=45, twistLimitDegrees=10}
 {bone1Name=dyn_ponytail_01_l, bone2Name=head, swing1LimitDegrees=45, swing2LimitDegrees=45, twistLimitDegrees=10}
 {bone1Name=dyn_ponytail_02_l, bone2Name=dyn_ponytail_01_l, swing1LimitDegrees=25, swing2LimitDegrees=25, twistLimitDegrees=10}
 {bone1Name=dyn_ponytail_02_r, bone2Name=dyn_ponytail_01_r, swing1LimitDegrees=25, swing2LimitDegrees=25, twistLimitDegrees=10}
 {bone1Name=dyn_ponytail_00_r, bone2Name=head, swing1LimitDegrees=35, swing2LimitDegrees=45, twistLimitDegrees=1}
 {bone1Name=dyn_ponytail_00_l, bone2Name=head, swing1LimitDegrees=35, swing2LimitDegrees=45, twistLimitDegrees=1}
 {bone1Name=dyn_hair_a_01_l, bone2Name=base_hair_a_01_l, swing1LimitDegrees=15, swing2LimitDegrees=25, twistLimitDegrees=5}
 {bone1Name=dyn_hair_a_02_l, bone2Name=dyn_hair_a_01_l, swing1LimitDegrees=15, swing2LimitDegrees=20, twistLimitDegrees=5}
 {bone1Name=dyn_hair_a_01_r, bone2Name=base_hair_a_01_r, swing1LimitDegrees=15, swing2LimitDegrees=25, twistLimitDegrees=10}
 {bone1Name=dyn_hair_a_02_r, bone2Name=dyn_hair_a_01_r, swing1LimitDegrees=15, swing2LimitDegrees=20, twistLimitDegrees=5}
 {bone1Name=dyn_hair_b_01_l, bone2Name=head, swing1LimitDegrees=45, swing2LimitDegrees=25, twistLimitDegrees=10}
 {bone1Name=dyn_hair_b_02_l, bone2Name=dyn_hair_b_01_l, swing1LimitDegrees=45, swing2LimitDegrees=25, twistLimitDegrees=10}
 {bone1Name=dyn_hair_b_01_r, bone2Name=head, swing1LimitDegrees=45, swing2LimitDegrees=25, twistLimitDegrees=10}
 {bone1Name=dyn_hair_b_02_r, bone2Name=dyn_hair_b_01_r, swing1LimitDegrees=45, swing2LimitDegrees=25, twistLimitDegrees=10}
 {
  bone1Name=base_hair_a_01_r
  bone2Name=head
  swing1LimitDegrees=45
  swing1Motion=Free
  swing2LimitDegrees=45
  swing2Motion=Free
  twistLimitDegrees=45
  twistMotion=Free
 }
 {
  bone1Name=base_hair_a_01_l
  bone2Name=head
  swing1LimitDegrees=45
  swing1Motion=Free
  swing2LimitDegrees=45
  swing2Motion=Free
  twistLimitDegrees=45
  twistMotion=Free
 }
 {bone1Name=dyn_hair_fr_l, bone2Name=head, swing1LimitDegrees=15, swing2LimitDegrees=15, twistLimitDegrees=10}
 {bone1Name=dyn_hair_fr_r, bone2Name=head, swing1LimitDegrees=15, swing2LimitDegrees=15, twistLimitDegrees=10}
 {
  bone1Name=clavicle_l
  bone2Name=spine_05
  swing1LimitDegrees=45
  swing1Motion=Free
  swing2LimitDegrees=45
  swing2Motion=Free
  twistLimitDegrees=45
  twistMotion=Free
 }
 {
  bone1Name=upperarm_l
  bone2Name=clavicle_l
  swing1LimitDegrees=45
  swing1Motion=Free
  swing2LimitDegrees=45
  swing2Motion=Free
  twistLimitDegrees=45
  twistMotion=Free
 }
 {
  bone1Name=clavicle_r
  bone2Name=spine_05
  swing1LimitDegrees=45
  swing1Motion=Free
  swing2LimitDegrees=45
  swing2Motion=Free
  twistLimitDegrees=45
  twistMotion=Free
 }
 {
  bone1Name=upperarm_r
  bone2Name=clavicle_r
  swing1LimitDegrees=45
  swing1Motion=Free
  swing2LimitDegrees=45
  swing2Motion=Free
  twistLimitDegrees=45
  twistMotion=Free
 }
 {
  bone1Name=neck_01
  bone2Name=spine_05
  swing1LimitDegrees=45
  swing1Motion=Free
  swing2LimitDegrees=45
  swing2Motion=Free
  twistLimitDegrees=45
  twistMotion=Free
 }
 {
  bone1Name=neck_02
  bone2Name=neck_01
  swing1LimitDegrees=45
  swing1Motion=Free
  swing2LimitDegrees=45
  swing2Motion=Free
  twistLimitDegrees=45
  twistMotion=Free
 }
]
