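# Physics asset for the pigtails on the RockerPunkAlt medium female player's face accessory mesh.
# Each side chains a small base capsule and three dynamic ponytail capsules from a larger head capsule.
# Base-to-head joints leave swing and twist free, while joints down each chain use angular limits.
# The sides are not fully mirrored: the left has tighter twist at its first dynamic joint and tighter swing at its tip.
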PhysicsAsset /Game/Characters/Player/Female/Medium/Bodies/F_MED_RockerPunkAlt/Meshes/Parts/F_MED_RockerPunkAlt_FaceAcc_Physics_pigtails
bodies {
 base_ponytail_00_l [{shapeType=Capsule, length=0.505, radius=0.505, shapeName=base_ponytail_00_l_capsule, rotation=<rot 0 -90 0>}]
 base_ponytail_00_r [{shapeType=Capsule, length=0.505, radius=0.505, shapeName=base_ponytail_00_r_capsule, rotation=<rot 0 -90 0>}]
 dyn_ponytail_00_l [{shapeType=Capsule, length=1, radius=1, shapeName=dyn_ponytail_00_l_capsule, center=<1 0 0>, rotation=<rot -90 0 0>}]
 dyn_ponytail_00_r [{shapeType=Capsule, length=1, radius=1, shapeName=dyn_ponytail_00_r_capsule, center=<1 0 0>, rotation=<rot -90 0 0>}]
 dyn_ponytail_01_l [{shapeType=Capsule, length=1, radius=1, shapeName=dyn_ponytail_01_l_capsule, center=<1 0 0>, rotation=<rot -90 0 0>}]
 dyn_ponytail_01_r [{shapeType=Capsule, length=1, radius=1, shapeName=dyn_ponytail_01_r_capsule, center=<1 0 0>, rotation=<rot -90 0 0>}]
 dyn_ponytail_02_l [{shapeType=Capsule, length=1, radius=1, shapeName=dyn_ponytail_02_l_capsule, center=<1 0 0>, rotation=<rot -90 0 0>}]
 dyn_ponytail_02_r [{shapeType=Capsule, length=1, radius=1, shapeName=dyn_ponytail_02_r_capsule, center=<1 0 0>, rotation=<rot -90 0 0>}]
 head [{shapeType=Capsule, length=3, radius=6.8, shapeName=head_capsule, center=<1 0 0>, rotation=<rot 90 0 0>}]
}
constraints [
 {
  bone1Name=base_ponytail_00_l
  bone2Name=head
  swing1LimitDegrees=45
  swing1Motion=Free
  swing2LimitDegrees=45
  swing2Motion=Free
  twistLimitDegrees=45
  twistMotion=Free
 }
 {bone1Name=dyn_ponytail_00_l, bone2Name=base_ponytail_00_l, swing1LimitDegrees=15, swing2LimitDegrees=15, twistLimitDegrees=1}
 {bone1Name=dyn_ponytail_01_l, bone2Name=dyn_ponytail_00_l, swing1LimitDegrees=45, swing2LimitDegrees=40, twistLimitDegrees=10}
 {bone1Name=dyn_ponytail_02_l, bone2Name=dyn_ponytail_01_l, swing1LimitDegrees=25, swing2LimitDegrees=20, twistLimitDegrees=10}
 {
  bone1Name=base_ponytail_00_r
  bone2Name=head
  swing1LimitDegrees=45
  swing1Motion=Free
  swing2LimitDegrees=45
  swing2Motion=Free
  twistLimitDegrees=45
  twistMotion=Free
 }
 {bone1Name=dyn_ponytail_00_r, bone2Name=base_ponytail_00_r, swing1LimitDegrees=15, swing2LimitDegrees=15, twistLimitDegrees=10}
 {bone1Name=dyn_ponytail_01_r, bone2Name=dyn_ponytail_00_r, swing1LimitDegrees=45, swing2LimitDegrees=40, twistLimitDegrees=10}
 {bone1Name=dyn_ponytail_02_r, bone2Name=dyn_ponytail_01_r, swing1LimitDegrees=45, swing2LimitDegrees=20, twistLimitDegrees=10}
]
