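# Post-process style AnimBlueprint for the RockerPunkAlt female face accessory, adding physics-driven hair motion over poses copied from the parent mesh, with clavicle and upper-arm branches layered from the body mesh.
# Normally, pigtail, back and front RigidBody sims chain from spine_05, then a hair bone scales when targeting with sufficient spine-to-head twist.
# Skydiving switches to a head-based sim, half-weighted under parachute, plus neck scaling.
# All sims share computed gravity override.
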AnimBlueprint /BRCosmetics/Characters/Player/Female/Medium/Bodies/F_MED_RockerPunkAlt/Meshes/Parts/F_MED_RockerPunkAlt_FaceAcc_AnimBP
target_skeleton=/Game/Characters/Player/Female/Medium/Bodies/F_MED_RockerPunkAlt/Meshes/Parts/F_MED_RockerPunkAlt_FaceAcc_Skeleton
graphs [
 AnimGraph {nodes [
   n1 = RigidBody!(
    ComponentPose=Local To Component(LocalPose=Use cached pose 'StartCache'())
    Alpha
    OverrideWorldGravity:Vector=Get GravityOverride()
    ExternalForce
   ) -> (Pose) {
    bClampLinearTranslationLimitToRefPose=true
    bOverrideWorldGravity=true
    overridePhysicsAsset=/Game/Characters/Player/Female/Medium/Bodies/F_MED_RockerPunkAlt/Meshes/Parts/F_MED_RockerPunkAlt_FaceAcc_Physics_pigtails
    simulationSpace=BaseBoneSpace
    alphaBoolBlend {blendOption=Linear}
    baseBoneRef {boneName=spine_05}
    componentAppliedLinearAccClamp=<3000 3000 3000>
    componentLinearAccScale=<0.22 0.22 0.5>
    componentLinearVelScale=<0.22 0.22 0.5>
    overrideWorldGravity {z=-980}
    simSpaceSettings {worldAlpha=0.25}
   }
   n2 = RigidBody(ComponentPose=@n1, OverrideWorldGravity=Get GravityOverride()) {
    bClampLinearTranslationLimitToRefPose=true
    bOverrideWorldGravity=true
    overridePhysicsAsset=/Game/Characters/Player/Female/Medium/Bodies/F_MED_RockerPunkAlt/Meshes/Parts/F_MED_RockerPunkAlt_FaceAcc_Physics_Back
    simulationSpace=BaseBoneSpace
    alphaBoolBlend {blendOption=Linear}
    alphaScaleBiasClamp {interpSpeedDecreasing=5, interpSpeedIncreasing=5}
    baseBoneRef {boneName=spine_05}
    componentAppliedLinearAccClamp=<3000 3000 3000>
    componentLinearAccScale=<0.22 0.22 0.2>
    componentLinearVelScale=<0.22 0.22 0.22>
    overrideWorldGravity {z=-980}
    simSpaceSettings {worldAlpha=0.2}
   }
   n3 = RigidBody(ComponentPose=@n2, OverrideWorldGravity=Get GravityOverride()) {
    bClampLinearTranslationLimitToRefPose=true
    bOverrideWorldGravity=true
    overridePhysicsAsset=/Game/Characters/Player/Female/Medium/Bodies/F_MED_RockerPunkAlt/Meshes/Parts/F_MED_RockerPunkAlt_FaceAcc_Physics_Front
    simulationSpace=BaseBoneSpace
    alphaBoolBlend {blendOption=Linear}
    alphaScaleBiasClamp {interpSpeedDecreasing=4, interpSpeedIncreasing=4}
    baseBoneRef {boneName=spine_05}
    componentAppliedLinearAccClamp=<3000 3000 3000>
    componentLinearAccScale=<0.22 0.22 0.22>
    componentLinearVelScale=<0.22 0.22 0.22>
    overrideWorldGravity {z=-980}
   }
   n4 = "Transform (Modify) Bone - Bone: dyn_hair_a_02_r"(
    ComponentPose=@n3
    Alpha=Get hair up + is targeting float() -> (hair up + is targeting float:double)
    Translation="0,0,0"
    Rotation="0,0,0"
    Scale="0.75,1,1"
   ) {
    scaleMode=BMM_Additive
    scaleSpace=BCS_BoneSpace
    alphaBoolBlend {blendOption=Linear}
    boneToModify {boneName=dyn_hair_a_02_r}
    scale {x=0.75}
   }
   n5 = RigidBody(
    ComponentPose=Local To Component(LocalPose=Use cached pose 'StartCache'())
    Alpha=Get is skydiving float() -> (is skydiving float:double)
    OverrideWorldGravity=Get GravityOverride()
   ) {
    bClampLinearTranslationLimitToRefPose=true
    bOverrideWorldGravity=true
    overridePhysicsAsset=&/F_MED_RockerPunkAlt_FaceAcc_Physics_Main
    simulationSpace=BaseBoneSpace
    alphaBoolBlend {blendOption=Linear}
    alphaScaleBiasClamp {bInterpResult=true, interpSpeedDecreasing=4, interpSpeedIncreasing=4}
    baseBoneRef {boneName=head}
    componentAppliedLinearAccClamp=<3000 3000 3000>
    componentLinearAccScale=<0.22 0.22 0.22>
    componentLinearVelScale=<0.22 0.22 0.22>
    overrideWorldGravity {z=-980}
    simSpaceSettings {worldAlpha=0.1}
   }
   n6 = "Transform (Modify) Bone - Bone: neck_02"(ComponentPose=@n5, Alpha=1, Translation="0,0,0", Rotation="0,0,0", Scale="1,1.05,1.087") {
    scaleMode=BMM_Additive
    scaleSpace=BCS_BoneSpace
    alphaBoolBlend {blendOption=Linear}
    boneToModify {boneName=neck_02}
    scale {y=1.05, z=1.087}
   }
   Output Pose(Result=Blend Poses by bool(BlendPose_0=Component To Local(ComponentPose=@n6), BlendPose_1=Component To Local(ComponentPose=@n4), bActiveValue=Get bIsSkyDiving()) {blendType=Linear}) {layerGroup=None}
   n7 = Layered blend per bone(
    BasePose=Copy Pose From Mesh() {bUseAttachedParent=true}
    BlendPoses_0=Copy Pose From Mesh(SourceMeshComponent=Get body mesh() -> (body mesh:Skeletal Mesh Component Object Reference)) {bUseAttachedParent=true}
   ) {layerSetup [{branchFilters [{boneName=clavicle_l}, {boneName=clavicle_r}, {boneName=upperarm_l}, {boneName=upperarm_r}]}]}
   Save cached pose 'StartCache'(Pose=@n7) {cacheName=StartCache}
  ]}
 EventGraph {nodes [
   Event BlueprintInitializeAnimation()
   > Set body mesh(body mesh:Skeletal Mesh Component Object Reference=GetOwnerSkeletalMeshComponent()) -> (Output_Get:Skeletal Mesh Component Object Reference)
   Event BlueprintUpdateAnimation()
   > n1 = Sequence!(execute) -> (then_0, then_1, then_2:Exec, then_3:Exec)
   > n2 = Branch(Condition=Get bIsParachuteOpened())
   > Set is skydiving float(is skydiving float:double=0.5) -> (Output_Get:double)
   Set is skydiving float(execute=@n2.else, is skydiving float=1)
   n3 = Gravity Override State Function DEPRECATED(
    execute=@n1.then_1
    CharacterPartAnimInstance=Self-Reference()
    GameTimeInSeconds=GetGameTimeInSeconds()
    GravityOverrideParams=Get Gravity Override Params()
   )
   > Set GravityOverride(GravityOverride=@n3)
   Set Hair Up(
    execute=@n1.then_2
    Hair Up:double=GetAlphaFromReferenceTransformDelta(SkeletalMeshComponent=Get body mesh(), BoneName=spine_05, BaseName=head, InputRange=(X=0,Y=55), TwistAxis="0,0,1")
   ) -> (Output_Get:double)
   n4 = Get bIsTargeting()
   > Set hair up + is targeting float(hair up + is targeting float:double=float * float(A="To Float (Boolean)"(InBool=@n4), B=Get hair up() -> (hair up:double)) {is=title K2Node_PromotableOperator}) -> (Output_Get:double)
  ]}
]
variables [
 body mesh {type=SkeletalMeshComponent}
 Gravity Override Params {type=GravityOverrideParamsStruct}
 GravityOverride {type=Vector}
 hair up {type=number}
 hair up + is targeting float {type=number}
 is jumping {type=number}
 is parachute opened float {type=number}
 is skydiving float {type=number}
 is sprinting alpha {type=number}
 is sprinting jumping alpha {type=number}
 is targeting alpha {type=number}
 is using weapons list {type=number}
 is using weapons list + hair up {type=number}
 Melee_float {type=number}
 RB_skydiving_alpha {type=number}
]
